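# Landscape material for the Apollo terrain in Creative, built on Material Attributes.
# A grass base blends six switchable layers (gravel, mud, sand, asphalt, forest floor, scree), then slope-based cliffs, with parallel mobile chains using cheaper variants.
# Opacity Mask carries blend data between layers before the visibility mask replaces it and the Creative gameplay function runs.
# Derived masks drive grass spawning and five prioritized physical materials, with debug blend parameters.
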Material /Game/Creative/Materials/Landscape/Creative_Landscape_Apollo/M_Creative_Landscape_Apollo
graph {
 properties {
  bTangentSpaceNormal=false
  bUseMaterialAttributes=true
  bUsedWithInstancedStaticMeshes=true
  bUsedWithNanite=true
  pixelDepthOffsetMode=PDOM_Legacy
  usedShadingModels=Default Lit
 }
 nodes [
  n1 = Max(A=LandscapeLayerSample() {parameterName=Layer1}, B=LandscapeLayerSample() {parameterName=Layer2})
  n2 = LandscapeLayerSample() {parameterName=Layer3}
  n3 = Max(A=Multiply(A=Max(A=@n1, B=@n2), B=Constant() {r=2}), B=Ceil(None=LandscapeLayerSample() {parameterName=Layer4}))
  n4 = Multiply(A=LandscapeLayerSample() {parameterName=Layer5}, B=Constant() {r=1.5})
  n5 = LandscapeLayerSample() {parameterName=Layer6}
  n6 = Constant() {r=2}
  n7 = Max(A=Max(A=Saturate(None=@n3), B=Saturate(None=@n4)), B=Saturate(None=Multiply(A=@n5, B=@n6)))
  n8 = NamedRerouteUsage()
  MaterialExpressionMultiply_7 = Multiply(A=OneMinus(None=Max(A=@n7, B=@n8)), B=LandscapeVisibilityMask())
  LandscapeGrassOutput!(Grass=@MaterialExpressionMultiply_7) {grassTypes [Grass {
     grassType=/Game/Athena/Apollo/Environments/Landscape/SmallFoliage/DataAssets/LandscapeGrassType/NonRVT/LGT_StandardGrass_NoRVT
     input {
      expression=&self:MaterialExpressionMultiply_7
      inputName=None
      mask=0
      maskA=0
      maskB=0
      maskG=0
      maskR=0
      outputIndex=0
     }
    }]}
  MaterialExpressionNamedRerouteUsage_64 = NamedRerouteUsage()
  MaterialExpressionNamedRerouteUsage_62 = NamedRerouteUsage()
  MaterialExpressionNamedRerouteUsage_63 = NamedRerouteUsage()
  MaterialExpressionNamedRerouteUsage_60 = NamedRerouteUsage()
  MaterialExpressionNamedRerouteUsage_61 = NamedRerouteUsage()
  LandscapePhysicalMaterialOutput!(
   Dirt=@MaterialExpressionNamedRerouteUsage_64
   Stone=@MaterialExpressionNamedRerouteUsage_62
   Grass=@MaterialExpressionNamedRerouteUsage_63
   Asphalt=@MaterialExpressionNamedRerouteUsage_60
   Sand=@MaterialExpressionNamedRerouteUsage_61
  ) {
   desc="These are layered in order of importance.\r\nIf Dirt=1 and Stone=1, then Dirt will \"win\".\r\nTheoretically, these values aren't clamped.\r\nSo a value of 2 should trump a value of 1."
   inputs [
    {
     physicalMaterial=/Game/FortressPhysicalMaterials/Dirt
     input {
      expression=&self:MaterialExpressionNamedRerouteUsage_64
      inputName=None
      mask=0
      maskA=0
      maskB=0
      maskG=0
      maskR=0
      outputIndex=0
     }
    }
    {
     physicalMaterial=/Game/FortressPhysicalMaterials/Stone
     input {
      expression=&self:MaterialExpressionNamedRerouteUsage_62
      inputName=None
      mask=0
      maskA=0
      maskB=0
      maskG=0
      maskR=0
      outputIndex=0
     }
    }
    {
     physicalMaterial=/Game/FortressPhysicalMaterials/Grass
     input {
      expression=&self:MaterialExpressionNamedRerouteUsage_63
      inputName=None
      mask=0
      maskA=0
      maskB=0
      maskG=0
      maskR=0
      outputIndex=0
     }
    }
    {
     physicalMaterial=/Game/FortressPhysicalMaterials/Asphalt
     input {
      expression=&self:MaterialExpressionNamedRerouteUsage_60
      inputName=None
      mask=0
      maskA=0
      maskB=0
      maskG=0
      maskR=0
      outputIndex=0
     }
    }
    {
     physicalMaterial=/Game/FortressPhysicalMaterials/Sand
     input {
      expression=&self:MaterialExpressionNamedRerouteUsage_61
      inputName=None
      mask=0
      maskA=0
      maskB=0
      maskG=0
      maskR=0
      outputIndex=0
     }
    }
   ]
  }
  n9 = Constant() {r=5}
  n10 = Constant() {r=-1}
  n11 = NamedRerouteUsage()
  n12 = LandscapeLayerSample() {parameterName=Layer1}
  n13 = MaterialFunctionCall!(LayerWeight=@n12, VarianceMask=NamedRerouteUsage(), BypassBiomeColors=StaticBool() {value=true}) {materialFunction=/Game/Athena/Apollo/Environments/Landscape/Materials/Biome/CommonMaterials/MF_Common_Gravel}
  n14 = MaterialFunctionCall!(
   CurrentMaterial=@n11
   AddMaterial=SetMaterialAttributes!(MaterialAttributes=@n13.Material, Opacity Mask=@n13.Material) {attributeSetTypes [#1]}
   LandscapeLayer=@n12
   UseContrast
   Contrast
  ) {materialFunction=/Game/Creative/Materials/Landscape/Creative_Landscape_Apollo/MaterialFunctions/MF_BlendLandscapeLayers_Opacity}
  NamedRerouteDeclaration(None=Saturate(None=LinearInterpolate(A=@n9, B=@n10, Alpha=@n14.Alpha))) {name=GM-Gravel, nodeColor=<rgba 0.5 0 1 1>}
  n15 = Max(
   A=Max(A=LandscapeLayerSample() {parameterName=Layer1}, B=LandscapeLayerSample() {parameterName=Layer2})
   B=LandscapeLayerSample() {parameterName=Layer3}
  )
  n16 = LandscapeLayerSample() {parameterName=Layer4}
  n17 = ComponentMask(None=VertexNormalWS()) {b=true}
  n18 = OneMinus(None=Power(Base=Abs(None=@n17), Exp=Constant() {r=3}))
  n19 = Constant() {r=0.3}
  n20 = Max(A=Max(A=Max(A=@n15, B=@n16), B=Multiply(A=@n18, B=@n19)), B=Constant() {desc=AdditionalGrassBorderColorMask})
  n21 = Constant() {r=2}
  n22 = StaticBool() {value=true}
  n23 = MaterialFunctionCall!(CombinedLandscapeLayersMask=Saturate(None=Multiply(A=@n20, B=@n21)), BypassUseRVT=@n22, BypassBiomeColors=@n22) {materialFunction=/Game/Athena/Apollo/Environments/Landscape/Materials/Biome/CommonMaterials/MF_Common_Grass}
  NamedRerouteDeclaration(None=@n23.GrassColorMask1) {name=GrassColorMask1, nodeColor=<rgba 0.5 0 1 1>}
  n24 = LandscapeLayerSample() {parameterName=Layer3}
  n25 = NamedRerouteUsage()
  n26 = MaterialFunctionCall!(UseUnderOceanSandMaterial, BypassBiomeColors=StaticBool() {value=true}, UseOceanLevel) {materialFunction=/Game/Athena/Apollo/Environments/Landscape/Materials/Biome/CommonMaterials/MF_Common_Sand}
  n27 = MaterialFunctionCall(
   CurrentMaterial=@n25
   AddMaterial=SetMaterialAttributes(MaterialAttributes=@n26.Material, Opacity Mask=@n26.Material) {attributeSetTypes [#1]}
   LandscapeLayer=@n24
  ) {materialFunction=&MF_BlendLandscapeLayers_Opacity}
  n28 = Saturate(None=LinearInterpolate(A=Constant() {r=30}, B=Constant(), Alpha=@n27.Alpha))
  NamedRerouteDeclaration(None=Saturate(None=Add(A=@n24, B=@n28))) {name=GM-Sand, nodeColor=<rgba 0.5 0 1 1>}
  n29 = LandscapeLayerSample() {desc=Asphalt, parameterName=Layer4}
  n30 = Constant() {r=3}
  NamedRerouteDeclaration(None=Saturate(None=Multiply(A=@n29, B=@n30))) {name=Asphalt, nodeColor=<rgba 1 1 0 1>}
  NamedRerouteDeclaration(None=LandscapeLayerSample() {desc=Sand, parameterName=Layer3}) {name=Sand, nodeColor=<rgba 1 1 0 1>}
  n31 = Constant() {r=5}
  n32 = Constant() {r=-1}
  n33 = NamedRerouteUsage()
  n34 = MaterialFunctionCall!(TilingSize) {materialFunction=/Game/Athena/Apollo/Environments/Landscape/Materials/MaterialFunctions/MF_ApolloAsphalt}
  n35 = MaterialFunctionCall(
   CurrentMaterial=@n33
   AddMaterial=SetMaterialAttributes(MaterialAttributes=@n34.DefaultLandscapeMaterial, Opacity Mask=@n34.DefaultLandscapeMaterial) {attributeSetTypes [#1]}
   LandscapeLayer=LandscapeLayerSample() {parameterName=Layer4}
  ) {materialFunction=&MF_BlendLandscapeLayers_Opacity}
  NamedRerouteDeclaration(None=Saturate(None=LinearInterpolate(A=@n31, B=@n32, Alpha=@n35.Alpha))) {name=GM-Asphalt, nodeColor=<rgba 0.5 0 1 1>}
  NamedRerouteDeclaration(None=LandscapeLayerSample() {desc=ForestFloor, parameterName=Layer5}) {name=FF, nodeColor=<rgba 1 1 0 1>}
  NamedRerouteDeclaration(None=LandscapeLayerSample() {desc=Scree, parameterName=Layer6}) {name=Scree, nodeColor=<rgba 1 1 0 1>}
  n36 = NamedRerouteUsage()
  n37 = Max(A=Max(A=NamedRerouteUsage(), B=NamedRerouteUsage()), B=NamedRerouteUsage())
  n38 = NamedRerouteUsage()
  n39 = Max(A=Saturate(None=Subtract(A=@n37, B=@n38)), B=NamedRerouteUsage())
  NamedRerouteDeclaration(None=Saturate(None=Subtract(A=@n36, B=@n39))) {name=Final-Grass, nodeColor=<rgba 0 1 1 1>}
  n40 = NamedRerouteUsage()
  n41 = Max(A=Max(A=NamedRerouteUsage(), B=NamedRerouteUsage()), B=NamedRerouteUsage())
  NamedRerouteDeclaration(None=Saturate(None=Subtract(A=@n40, B=@n41))) {name=Final-Asphalt, nodeColor=<rgba 0 1 1 1>}
  n42 = Max(A=NamedRerouteUsage(), B=NamedRerouteUsage())
  n43 = NamedRerouteUsage()
  n44 = Saturate(None=Subtract(A=NamedRerouteUsage(), B=Max(A=@n42, B=@n43)))
  n45 = NamedRerouteUsage()
  NamedRerouteDeclaration(None=Saturate(None=Subtract(A=@n44, B=@n45))) {name=Final-Sand, nodeColor=<rgba 0 1 1 1>}
  NamedRerouteDeclaration(None=Max(A=NamedRerouteUsage(), B=NamedRerouteUsage())) {name=Final-Stone, nodeColor=<rgba 0 1 1 1>}
  n46 = GetMaterialAttributes(None=NamedRerouteUsage()) {attributeGetTypes [#1]}
  n47 = MaterialFunctionCall!(
   Material=@n46.MaterialAttributes
   Heightmap=@n46.MaterialAttributes
   NewSlopeMaskMask
   Displacement
   AsphaltLayer
   UseAlternateCliffMat
   AltCliffHeight
   AltCliffMat
  ) {materialFunction=/Game/Athena/Apollo/Environments/Landscape/Materials/Biome/CommonMaterials/MF_SlopeBlendCliff}
  NamedRerouteDeclaration(None=@n47.LandscapeGrassMeshMask) {name=CliffGrassMeshMask, nodeColor=<rgba 0.5 0 1 1>}
  n48 = NamedRerouteUsage()
  n49 = Max(A=NamedRerouteUsage(), B=NamedRerouteUsage())
  n50 = NamedRerouteUsage()
  n51 = Max(A=Max(A=Max(A=@n49, B=@n50), B=NamedRerouteUsage()), B=NamedRerouteUsage())
  NamedRerouteDeclaration(None=Saturate(None=Subtract(A=@n48, B=@n51))) {name=Final-Dirt, nodeColor=<rgba 0 1 1 1>}
  NamedRerouteDeclaration(None=LandscapeLayerSwitch!(
    LayerUsed=SetMaterialAttributes(MaterialAttributes=@n14.BlendedMaterial, Opacity Mask=@n14.BlendedMaterial) {attributeSetTypes [#1]}
    LayerNotUsed=@n11
   ) {parameterName=Layer1}) {name=With_Layer1, nodeColor=<rgba 0.5 0 1 1>}
  n52 = NamedRerouteUsage()
  n53 = MaterialFunctionCall!(
   Diffuse=@n13.Diffuse
   Height=@n13.Diffuse
   Spec=@n13.Diffuse
   Rough=@n13.Diffuse
   Normals=@n13.Diffuse
   VarianceMask=NamedRerouteUsage()
   BiomeRoughMod
   BypassBiomeColors=StaticBool() {value=true}
  ) {materialFunction=/Game/Athena/Apollo/Environments/Landscape/Materials/Biome/CommonMaterials/MF_Common_Mud}
  n54 = MaterialFunctionCall(
   CurrentMaterial=@n52
   AddMaterial=SetMaterialAttributes(MaterialAttributes=@n53.Material, Opacity Mask=@n53.Material) {attributeSetTypes [#1]}
   LandscapeLayer=LandscapeLayerSample() {parameterName=Layer2}
  ) {materialFunction=&MF_BlendLandscapeLayers_Opacity}
  NamedRerouteDeclaration(None=LandscapeLayerSwitch(
    LayerUsed=SetMaterialAttributes(MaterialAttributes=@n54.BlendedMaterial, Opacity Mask=@n54.BlendedMaterial) {attributeSetTypes [#1]}
    LayerNotUsed=@n52
   ) {parameterName=Layer2}) {name=With_Layer2, nodeColor=<rgba 0.5 0 1 1>}
  NamedRerouteDeclaration(None=LandscapeLayerSwitch(
    LayerUsed=SetMaterialAttributes(MaterialAttributes=@n27.BlendedMaterial, Opacity Mask=@n27.BlendedMaterial) {attributeSetTypes [#1]}
    LayerNotUsed=@n25
   ) {parameterName=Layer3}) {name=With_Layer3, nodeColor=<rgba 0.5 0 1 1>}
  NamedRerouteDeclaration(None=LandscapeLayerSwitch(
    LayerUsed=SetMaterialAttributes(MaterialAttributes=@n35.BlendedMaterial, Opacity Mask=@n35.BlendedMaterial) {attributeSetTypes [#1]}
    LayerNotUsed=@n33
   ) {parameterName=Layer4}) {name=With_Layer4, nodeColor=<rgba 0.5 0 1 1>}
  n55 = NamedRerouteUsage()
  n56 = MaterialFunctionCall(
   CurrentMaterial=@n55
   AddMaterial=NamedRerouteUsage()
   LandscapeLayer=LandscapeLayerSample() {parameterName=Layer5}
   UseContrast=StaticBool() {value=true}
   Contrast=Constant() {r=0.5}
  ) {materialFunction=&MF_BlendLandscapeLayers_Opacity}
  NamedRerouteDeclaration(None=LandscapeLayerSwitch(
    LayerUsed=SetMaterialAttributes(MaterialAttributes=@n56.BlendedMaterial, Opacity Mask=@n56.BlendedMaterial) {attributeSetTypes [#1]}
    LayerNotUsed=@n55
   ) {parameterName=Layer5}) {name=With_Layer5, nodeColor=<rgba 0.5 0 1 1>}
  n57 = NamedRerouteUsage()
  n58 = MaterialFunctionCall(
   CurrentMaterial=@n57
   AddMaterial=NamedRerouteUsage()
   LandscapeLayer=LandscapeLayerSample() {parameterName=Layer6}
   UseContrast=StaticBool() {value=true}
   Contrast=Constant() {r=0.5}
  ) {materialFunction=&MF_BlendLandscapeLayers_Opacity}
  NamedRerouteDeclaration(None=LandscapeLayerSwitch(
    LayerUsed=SetMaterialAttributes(MaterialAttributes=@n58.BlendedMaterial, Opacity Mask=@n58.BlendedMaterial) {attributeSetTypes [#1]}
    LayerNotUsed=@n57
   ) {parameterName=Layer6}) {name=With_Layer6, nodeColor=<rgba 0.5 0 1 1>}
  NamedRerouteDeclaration(None=SetMaterialAttributes(MaterialAttributes=@n23.Default-LandscapeMaterial, Opacity Mask=@n23.Default-LandscapeMaterial) {attributeSetTypes [#1]}) {name=BaseMaterial, nodeColor=<rgba 0.5 0 1 1>}
  n59 = LandscapeLayerSample() {parameterName=Layer1}
  n60 = LandscapeLayerSample() {parameterName=Layer2}
  n61 = Max(A=Max(A=Max(A=@n59, B=@n60), B=LandscapeLayerSample() {parameterName=Layer3}), B=LandscapeLayerSample() {parameterName=Layer4})
  n62 = Constant() {r=3}
  n63 = StaticBool() {value=true}
  n64 = MaterialFunctionCall!(CombinedLandscapeLayersMask=Saturate(None=Multiply(A=@n61, B=@n62)), BypassUseRVT=@n63, BypassBiomeColors=@n63) {materialFunction=/Game/Athena/Apollo/Environments/Landscape/Materials/Biome/CommonMaterials/Mobile/MF_Mobile_Common_Grass}
  NamedRerouteDeclaration(None=SetMaterialAttributes(MaterialAttributes=@n64.MobileLandscapeMaterial, Opacity Mask=@n64.MobileLandscapeMaterial) {attributeSetTypes [#1]}) {name=Mobile_BaseMaterial, nodeColor=<rgba 0.5 0 1 1>}
  NamedRerouteDeclaration(None=@n47.CliffMask) {name=CliffPhysicalMatMask, nodeColor=<rgba 0.5 0 1 1>}
  n65 = NamedRerouteUsage()
  n66 = LandscapeLayerSample() {parameterName=Layer1}
  n67 = MaterialFunctionCall!(OverrideEdgeColor, EdgeMask=@n66, BypassBiomeColors=StaticBool() {value=true}) {materialFunction=/Game/Athena/Apollo/Environments/Landscape/Materials/Biome/CommonMaterials/Mobile/MF_Mobile_Common_Gravel}
  n68 = SetMaterialAttributes(MaterialAttributes=@n67.Material, Opacity Mask=@n67.Material) {attributeSetTypes [#1]}
  n69 = MaterialFunctionCall(CurrentMaterial=@n65, AddMaterial=@n68, LandscapeLayer=@n66) {materialFunction=&MF_BlendLandscapeLayers_Opacity}
  NamedRerouteDeclaration(None=LandscapeLayerSwitch(
    LayerUsed=SetMaterialAttributes(MaterialAttributes=@n69.BlendedMaterial, Opacity Mask=@n69.BlendedMaterial) {attributeSetTypes [#1]}
    LayerNotUsed=@n65
   ) {parameterName=Layer1}) {name=Mobile_With_Layer1, nodeColor=<rgba 0.5 0 1 1>}
  n70 = NamedRerouteUsage()
  n71 = StaticBool() {value=true}
  n72 = MaterialFunctionCall(
   CurrentMaterial=@n70
   AddMaterial=SetMaterialAttributes!(
    MaterialAttributes=@n68
    Base Color=MaterialFunctionCall!(GravelMaterialInput, Heightmap, UseUnmodifiedGravelDiffuse=@n71, UnmodifiedGravelDiffuse=@n67.RGB, BypassBiomeColors=@n71) {materialFunction=/Game/Athena/Apollo/Environments/Terrain_Meshes/MaterialFunctions/MF_DirtMudModifier}.ColorizedGravelDiffuse
   ) {attributeSetTypes [#2]}
   LandscapeLayer=LandscapeLayerSample() {parameterName=Layer2}
  ) {materialFunction=&MF_BlendLandscapeLayers_Opacity}
  NamedRerouteDeclaration(None=LandscapeLayerSwitch(
    LayerUsed=SetMaterialAttributes!(MaterialAttributes=@n72.BlendedMaterial, Opacity Mask=@n72.BlendedMaterial) {attributeSetTypes [#1]}
    LayerNotUsed=@n70
   ) {parameterName=Layer2}) {name=Mobile_With_Layer2, nodeColor=<rgba 0.5 0 1 1>}
  n73 = NamedRerouteUsage()
  n74 = MaterialFunctionCall!(CustomSeaLevel=@n26.WaterLevel, BypassBiomeColors=StaticBool() {value=true}, UseOceanLevel) {materialFunction=/Game/Athena/Apollo/Environments/Landscape/Materials/Biome/CommonMaterials/Mobile/MF_Mobile_Common_Sand}
  n75 = MaterialFunctionCall(
   CurrentMaterial=@n73
   AddMaterial=SetMaterialAttributes(MaterialAttributes=@n74.Material, Opacity Mask=@n74.Material) {attributeSetTypes [#1]}
   LandscapeLayer=LandscapeLayerSample() {parameterName=Layer3}
  ) {materialFunction=&MF_BlendLandscapeLayers_Opacity}
  NamedRerouteDeclaration(None=LandscapeLayerSwitch(
    LayerUsed=SetMaterialAttributes(MaterialAttributes=@n75.BlendedMaterial, Opacity Mask=@n75.BlendedMaterial) {attributeSetTypes [#1]}
    LayerNotUsed=@n73
   ) {parameterName=Layer3}) {name=Mobile_With_Layer3, nodeColor=<rgba 0.5 0 1 1>}
  n76 = NamedRerouteUsage()
  n77 = MaterialFunctionCall() {materialFunction=&MF_ApolloAsphalt}
  n78 = MaterialFunctionCall(
   CurrentMaterial=@n76
   AddMaterial=SetMaterialAttributes(MaterialAttributes=@n77.MobileLandscapeMaterial, Opacity Mask=@n77.MobileLandscapeMaterial) {attributeSetTypes [#1]}
   LandscapeLayer=LandscapeLayerSample() {parameterName=Layer4}
  ) {materialFunction=&MF_BlendLandscapeLayers_Opacity}
  NamedRerouteDeclaration(None=LandscapeLayerSwitch(
    LayerUsed=SetMaterialAttributes(MaterialAttributes=@n78.BlendedMaterial, Opacity Mask=@n78.BlendedMaterial) {attributeSetTypes [#1]}
    LayerNotUsed=@n76
   ) {parameterName=Layer4}) {name=Mobile_With_Layer4, nodeColor=<rgba 0.5 0 1 1>}
  n79 = LandscapeLayerCoords() {mappingScale=10}
  n80 = TextureSample(UVs=@n79) {samplerSource=SSM_Wrap_WorldGroupSettings, texture=/Game/Environments/Apollo/Sets/Terrain/Textures/T_Apollo_Forest_Floor_D_01}
  n81 = TextureSample(UVs=@n79) {
   samplerSource=SSM_Wrap_WorldGroupSettings
   samplerType=SAMPLERTYPE_Masks
   texture=/Game/Environments/Apollo/Sets/Terrain/Textures/T_Apollo_Forest_Floor_S_01
  }
  n82 = TextureSample(UVs=@n79) {
   samplerSource=SSM_Wrap_WorldGroupSettings
   samplerType=SAMPLERTYPE_Normal
   texture=/Game/Environments/Apollo/Sets/Terrain/Textures/T_Apollo_Forest_Floor_N_01
  }
  n83 = SetMaterialAttributes!(MaterialAttributes, Base Color=@n80.RGB, Specular=@n81.R, Roughness=@n81.R, Normal=@n82.RGB, Opacity Mask=@n80.RGB) {attributeSetTypes [
    #2
    #3
    #4
    #5
    #1
   ]}
  NamedRerouteDeclaration(None=@n83) {name=Layer5_Material, nodeColor=<rgba 1 1 0 1>}
  n84 = NamedRerouteUsage()
  n85 = MaterialFunctionCall(
   CurrentMaterial=@n84
   AddMaterial=NamedRerouteUsage()
   LandscapeLayer=LandscapeLayerSample() {parameterName=Layer5}
   UseContrast=StaticBool() {value=true}
   Contrast=Constant() {r=0.5}
  ) {materialFunction=&MF_BlendLandscapeLayers_Opacity}
  NamedRerouteDeclaration(None=LandscapeLayerSwitch(
    LayerUsed=SetMaterialAttributes!(MaterialAttributes=@n85.BlendedMaterial, Opacity Mask=@n85.BlendedMaterial) {attributeSetTypes [#1]}
    LayerNotUsed=@n84
   ) {parameterName=Layer5}) {name=Mobile_With_Layer5, nodeColor=<rgba 0.5 0 1 1>}
  n86 = SetMaterialAttributes!(
   MaterialAttributes
   Base Color=@n80.RGB
   Specular=Constant() {r=0.5}
   Roughness=Constant() {r=0.9}
   Normal=VertexNormalWS()
   Opacity Mask=@n80.RGB
  ) {attributeSetTypes [
    #2
    #3
    #4
    #5
    #1
   ]}
  NamedRerouteDeclaration(None=@n86) {name=Mobile_Layer5_Material, nodeColor=<rgba 1 1 0 1>}
  n87 = Multiply(A=LandscapeLayerCoords() {mappingScale=1}, B=Constant() {r=0.08})
  n88 = TextureSample(UVs=@n87) {samplerSource=SSM_Wrap_WorldGroupSettings, texture=/Game/Environments/Apollo/Sets/Terrain/Textures/T_Apollo_Scree_D_01}
  n89 = Saturate(None=Divide(A=@n88.RGB, B=Constant3Vector() {constant=<rgba 1.04 1.2884 1.6 1>}))
  n90 = SetMaterialAttributes(Base Color=@n89, Specular=Constant() {r=0.5}, Roughness=Constant() {r=0.9}, Normal=VertexNormalWS(), Opacity Mask=@n88.A) {attributeSetTypes [
    #2
    #3
    #4
    #5
    #1
   ]}
  NamedRerouteDeclaration(None=@n90) {name=Mobile_Layer6_Material, nodeColor=<rgba 1 1 0 1>}
  n91 = TextureSample(UVs=Multiply(A=LandscapeLayerCoords() {mappingScale=1}, B=Constant() {r=0.2})) {samplerSource=SSM_Wrap_WorldGroupSettings, texture=/Game/Environments/Apollo/Rocks/Natural_Rocks/Textures/T_RockDetails_02D}
  n92 = Multiply(A=@n89, B=SquareRoot(None=@n91.RGB))
  n93 = Constant() {r=1.5}
  n94 = Constant() {r=2.05}
  n95 = Saturate(None=Add(A=Constant() {r=-1.1}, B=Multiply(A=@n94, B=@n91.R)))
  n96 = Saturate(None=LinearInterpolate(A=Constant() {r=7}, B=Constant() {r=-0.8}, Alpha=@n91.R))
  n97 = TextureSample(UVs=@n87) {
   samplerSource=SSM_Wrap_WorldGroupSettings
   samplerType=SAMPLERTYPE_Masks
   texture=/Game/Environments/Apollo/Sets/Terrain/Textures/T_Apollo_Scree_S_01
  }
  n98 = Saturate(None=LinearInterpolate(A=Constant() {r=-0.25}, B=Constant() {r=1.5}, Alpha=@n97.R))
  n99 = Constant() {r=0.5}
  n100 = TextureSample(UVs=@n87) {
   samplerSource=SSM_Wrap_WorldGroupSettings
   samplerType=SAMPLERTYPE_Normal
   texture=/Game/Environments/Apollo/Sets/Terrain/Textures/T_Apollo_Scree_N_01
  }
  n101 = SetMaterialAttributes(
   Base Color=Saturate(None=Multiply(A=@n92, B=@n93))
   Specular=Multiply(A=Multiply(A=@n95, B=@n96), B=Clamp(None=@n98, Max=@n99) {clampMode=CMODE_ClampMax})
   Roughness=@n97.B
   Normal=@n100.RGB
   Opacity Mask=@n88.A
  ) {attributeSetTypes [
    #2
    #3
    #4
    #5
    #1
   ]}
  NamedRerouteDeclaration(None=@n101) {name=Layer6_Material, nodeColor=<rgba 1 1 0 1>}
  n102 = NamedRerouteUsage()
  n103 = MaterialFunctionCall(
   CurrentMaterial=@n102
   AddMaterial=NamedRerouteUsage()
   LandscapeLayer=LandscapeLayerSample() {parameterName=Layer6}
   UseContrast=StaticBool() {value=true}
   Contrast=Constant() {r=0.5}
  ) {materialFunction=&MF_BlendLandscapeLayers_Opacity}
  NamedRerouteDeclaration(None=LandscapeLayerSwitch(
    LayerUsed=SetMaterialAttributes!(MaterialAttributes=@n103.BlendedMaterial, Opacity Mask=@n103.BlendedMaterial) {attributeSetTypes [#1]}
    LayerNotUsed=@n102
   ) {parameterName=Layer6}) {name=Mobile_With_Layer6, nodeColor=<rgba 0.5 0 1 1>}
  n104 = LinearInterpolate(
   A=LinearInterpolate(A=NamedRerouteUsage(), B=NamedRerouteUsage(), Alpha=ScalarParameter() {parameterName=DEBUG-Stone})
   B=NamedRerouteUsage()
   Alpha=ScalarParameter() {parameterName=DEBUG-Grass}
  )
  n105 = NamedRerouteUsage()
  n106 = ScalarParameter() {parameterName=DEBUG-Asphalt}
  NamedRerouteDeclaration(None=LinearInterpolate(A=LinearInterpolate(A=@n104, B=@n105, Alpha=@n106), B=NamedRerouteUsage(), Alpha=ScalarParameter() {parameterName=DEBUG-Sand})) {name=DEBUG-PhysMat, nodeColor=<rgba 0.5 0 1 1>}
  NamedRerouteDeclaration(None=@n47.HeightmapForGrass) {name=HeightmapForGrass, nodeColor=<rgba 0.5 0 1 1>}
  n107 = Max(A=LandscapeLayerSample() {desc=Gravel, parameterName=Layer1}, B=LandscapeLayerSample() {desc=Mud, parameterName=Layer2})
  n108 = Constant() {r=2}
  NamedRerouteDeclaration(None=Saturate(None=Multiply(A=@n107, B=@n108))) {name=Gravel+Mud, nodeColor=<rgba 1 1 0 1>}
  n109 = Constant() {r=5}
  n110 = Constant() {r=-1}
  NamedRerouteDeclaration(None=Saturate(None=LinearInterpolate(A=@n109, B=@n110, Alpha=@n54.Alpha))) {name=GM-Mud, nodeColor=<rgba 0.5 0 1 1>}
  n111 = NamedRerouteUsage()
  NamedRerouteDeclaration(None=OneMinus(None=@n111)) {name="Cliff=White", nodeColor=<rgba 1 1 0 1>}
  NamedRerouteDeclaration(None=@n111) {name="Cliff=Black", nodeColor=<rgba 1 1 0 1>}
  NamedRerouteUsage() {desc=Won't match the grass mesh color}
  n112 = GetMaterialAttributes(None=NamedRerouteUsage()) {attributeGetTypes [#1]}
  n113 = SetMaterialAttributes!(MaterialAttributes=@n112.MaterialAttributes, Normal=VertexNormalWS()) {attributeSetTypes [#5]}
  n114 = MaterialFunctionCall!(UseParameters=StaticBool()) {materialFunction=/Game/Athena/Apollo/Environments/Landscape/Materials/Biome/CommonMaterials/Mobile/MF_Mobile_Common_Cliff}
  n115 = SetMaterialAttributes!(
   MaterialAttributes=ShadingPathSwitch(
    Default=MaterialFunctionCall!(Material=@n47.Material) {materialFunction=/Game/Athena/Apollo/Maps/Landscape/VT/MaterialFunctions/TransformTangentToWorld}.Result
    Mobile=MaterialFunctionCall!(
     Material 1=@n113
     Height Texture 1=@n112.Opacity Mask
     Material 2=@n114.LandscapeMaterial
     Height Texture 2=@n114.LandscapeMaterial
     Transistion Phase=@n114.LandscapeMaterial
     UseContrast
     Contrast
     AddLayerMask
    ) {materialFunction=/Game/Athena/Apollo/Environments/Landscape/Materials/Biome/CommonMaterials/MF_BlendLandscapeLayers}.BlendedMaterial
   ) {desc="Mobile path normals are already in worldspace!"}
   Opacity Mask=LandscapeVisibilityMask()
  ) {attributeSetTypes [#1]}
  n116 = StaticBool() {value=true}
  SetMaterialAttributes!(
   MaterialAttributes=GetMaterialAttributes(None=MaterialFunctionCall!(In=@n115, IsForLandscapeMaterial=@n116) {materialFunction=/Game/Creative/Materials/MaterialFunctions/MF_CreativeFortniteGameplay}.Result) {attributeGetTypes [#2]}.MaterialAttributes
   Base Color=NamedRerouteUsage()
  ) {attributeSetTypes [#2]}
 ]
}
